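; Spring event Heal Egg gameplay ability that heals the player's health first, then shields, up to max.
; Each EventApplyEffect applies the matching GameplayEffect, then cancels the ability if ForceCancelAbility returns true.
; RepNotify flags toggle between health and shield, re-adding the heal-over-time GameplayCue with source tags that select its VFX.
; Activation starts that cue immediately to avoid visual delay, and start and end burst cues bracket the ability.
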(event Ability|EventFailedToActivatePassiveAbility
  (params (ActorInfo :type Gameplay%20Ability%20Actor%20Info%20Structure))
  ;[1 0 0.6156] Event FailedToActivatePassiveAbility
  ; Must be called for GAs using ActivateAbility
  (|CallParentFunction ActorInfo))

(event Ability|EventActivateAbility
  ;[0.009722 1 0] Event ActivateAbility
  ; Activate the Healing GCN on activation to prevent a short visual delay from the initial activate to the first healing period
  (|CallParentFunction)
  (CallFunction|ActivateHealingOverTimeGCN (CallFunction|GetHealingGCNTag))
  (Ability|ExecuteGameplayCueWithParamsOnOwner (Variables|Default|GetGCNTag_StartBurst) (Ability|GameplayCue|MakeGameplayCueParameters :NormalizedMagnitude 0.0 :RawMagnitude 0.0 :OriginalTag (Variables|Default|GetGCNTag_StartBurst) :Location (Transformation|GetActorLocation (Ability|GetAvatarActorfromActorInfo)) :Normal "0.000000,0.000000,1.000000" :GameplayEffectLevel 1 :AbilityLevel 1 :bReplicateLocationWhenUsingMinimalRepProxy false)))

(event EventApplyEffect
  (bind _playerpawn (Variables|Default|GetPlayerPawn))
  ;[0 0.03565 1] EventApplyEffect
  ; Apply healing sheilds or health up to max starting with health. Update the GCN call to pass a parameter for each type of healing.
  (Utilities|IsValid _playerpawn
    (:Is%20Valid
      (if (== (FortPawn|GetMaxHealth _playerpawn) (FortPawn|GetHealth _playerpawn))
        (if (not (== (FortPawn|GetShield _playerpawn) (FortPawn|GetMaxShield _playerpawn)))
          (Ability|ApplyGameplayEffectToOwner (Variables|Default|GetGE_Shields) 0)
          (Variables|Default|SetbIsHealingShield true))
        (else
          (Ability|ApplyGameplayEffectToOwner (Variables|Default|GetGE_Heals) 0)
          (Variables|Default|SetbIsHealingHealth true)))
      (Default|CallOnEffectApplied)
      (if (CallFunction|ForceCancelAbility)
        (Ability|CancelAbility)))))

(event Ability|EventOnEndAbility
  (params (bWasCancelled :type Boolean))
  ;[1 0.4762 0] OnEndAbility
  ; Call End Burst VFX
  (|CallParentFunction bWasCancelled)
  (Ability|ExecuteGameplayCueWithParamsOnOwner (Variables|Default|GetGCNTag_EndBurst) (Ability|GameplayCue|MakeGameplayCueParameters :NormalizedMagnitude 0.0 :RawMagnitude 0.0 :OriginalTag (Variables|Default|GetGCNTag_EndBurst) :AggregatedSourceTags (select (Variables|Default|GetbIsHealingHealth) (Variables|Default|GetGCSourceTags_HealingHP) (Variables|Default|GetGCSourceTags_HealingShield)) :Location (Transformation|GetActorLocation (Ability|GetAvatarActorfromActorInfo)) :Normal "0.000000,0.000000,1.000000" :GameplayEffectLevel 1 :AbilityLevel 1 :bReplicateLocationWhenUsingMinimalRepProxy false)))

(fn ActivateHealingOverTimeGCN
  (params (AggregatedSourceTags :type Gameplay%20Tag%20Container%20Structure))
  ;[0.5183 0 1] ActivateHealingOverTimeGCN
  ; Refresh and Add the GameplayCue with a tag identifying the type of healing for the VFX parameters
  (Utilities|FlowControl|Sequence
    (:then_0
      ;;[1 0.3891 0] [0] Remove the effect since it won't reapply an acitve GCN
      (Ability|RemoveGameplayCueFromOwner (Variables|Default|GetGCNTag_HealOverTime)))
    (:then_1
      ;;[0.5183 0 1] [1] Apply a GCN with the appropriately set tags to determine healing sheilds/health for VFX paramters
      (Ability|AddGameplayCueWithParamsToOwner (Variables|Default|GetGCNTag_HealOverTime) (Ability|GameplayCue|MakeGameplayCueParameters :NormalizedMagnitude 0.0 :RawMagnitude 0.0 :OriginalTag (Variables|Default|GetGCNTag_HealOverTime) :AggregatedSourceTags AggregatedSourceTags :Location (Transformation|GetActorLocation (Ability|GetAvatarActorfromActorInfo)) :Normal "0.000000,0.000000,1.000000" :GameplayEffectLevel 1 :AbilityLevel 1 :bReplicateLocationWhenUsingMinimalRepProxy false)))))

(fn GetHealingGCNTag
  (if (== (FortPawn|GetMaxHealth (Variables|Default|GetPlayerPawn)) (FortPawn|GetHealth (Variables|Default|GetPlayerPawn)))
    (return (Variables|Default|GetGCSourceTags_HealingShield))
    (else
      (return (Variables|Default|GetGCSourceTags_HealingHP)))))

(fn OnRep_bIsHealingHealth
  ;[0.6857 0 1] OnRepB IsHealingHealth
  ; Check if the healing effect is not yet active, disable shields, and activate the healing GCN only when a toggle occurs
  (if (Variables|Default|GetbIsHealingHealth)
    (if (Variables|Default|GetbIsHealingShield)
      (Variables|Default|SetbIsHealingShield false)
      (elif (Variables|Default|GetbIsHealingStarted)
        _
        (else
          ;;[0.2094 0.6678 1] Check if this is the first run and activate healing for the first time
          (Variables|Default|SetbIsHealingStarted true))))
    (CallFunction|ActivateHealingOverTimeGCN (Variables|Default|GetGCSourceTags_HealingHP))))

(fn OnRep_bIsHealingShield
  ;[0.6857 0 1] OnRepB IsHealingShield
  ; Check if the healing effect is not yet active, disable shields, and activate the healing GCN only when a toggle occurs
  (if (Variables|Default|GetbIsHealingShield)
    (if (Variables|Default|GetbIsHealingHealth)
      (Variables|Default|SetbIsHealingHealth false)
      (elif (Variables|Default|GetbIsHealingStarted)
        _
        (else
          ;;[0.2094 0.6678 1] Check if this is the first run and activate healing for the first time
          (Variables|Default|SetbIsHealingStarted true))))
    (CallFunction|ActivateHealingOverTimeGCN (Variables|Default|GetGCSourceTags_HealingShield))))
